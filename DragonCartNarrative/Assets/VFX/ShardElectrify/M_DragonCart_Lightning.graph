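# Unlit, two-sided additive material for lightning effects on Niagara sprites and ribbons, used for the DragonCart shard electrify VFX.
# A tiling, panning beam texture has a saturated sum of a separately tiled noise texture and the Erosion dynamic parameter subtracted from it, letting particles erode the bolts.
# The eroded mask multiplies particle color, and also particle alpha, which passes through a depth fade to soften intersections.
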Material /DragonCartNarrative/Assets/VFX/ShardElectrify/M_DragonCart_Lightning
graph {
 properties {
  bAutomaticallySetUsageInEditor=false
  bUsedWithNiagaraRibbons=true
  bUsedWithNiagaraSprites=true
  blendMode=BLEND_Additive
  shadingModel=MSM_Unlit
  twoSided=true
  usedShadingModels=Unlit
 }
 nodes [
  n1 = ParticleColor()
  n2 = TextureCoordinate()
  n3 = AppendVector(A=ScalarParameter() {defaultValue=1, parameterName=Emissive_U_Freq}, B=ScalarParameter() {defaultValue=1, parameterName=Emissive_V_Freq})
  n4 = ScalarParameter() {parameterName=Emissive_U_Speed}
  n5 = ScalarParameter() {parameterName=Emissive_V_Speed}
  n6 = TextureSampleParameter2D(UVs=Panner(Coordinate=Multiply(A=@n2, B=@n3), Speed=AppendVector(A=@n4, B=@n5))) {parameterName=Lightning_Texture, samplerType=SAMPLERTYPE_LinearColor, texture=/Game/Effects/Fort_Effects/Textures/Tile/T_Raygun_Beam_02}
  n7 = ScalarParameter() {defaultValue=1, parameterName=Erosion_U_Freq}
  n8 = ScalarParameter() {defaultValue=1, parameterName=Erosion_V_Freq}
  n9 = TextureSampleParameter2D(UVs=Multiply(A=TextureCoordinate(), B=AppendVector(A=@n7, B=@n8))) {
   parameterName=Erosion_Texture
   samplerType=SAMPLERTYPE_Grayscale
   texture=&/T_DragonCart_LightningNoise
  }
  n10 = DynamicParameter() {paramNames [Erosion,Param2,Param3,Param4]}
  n11 = Subtract(A=@n6.R, B=Saturate(None=Add(A=@n9.R, B=@n10.Erosion) {constB=-1}))
  DepthFade(Opacity=Multiply(A=@n1.A, B=@n11), FadeDistance=ScalarParameter() {parameterName=DepthFade})
  Multiply(A=@n1.RGB, B=@n11)
 ]
}
