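# Niagara system for long electric arcs and sparks, named for the electrified shards in DragonCart, with User.Intensity driving rate, lifetime and brightness.
# Lightning_01 spawns short-lived bolts using user-set volume, scale and orientation, rendered through mesh, sprite and light renderers with animated material parameters.
# Sparks001 reads those bolts to emit falling, speed-stretched sparks that play zap audio on spawn at User.ZapVolume.
# Both emitters scale spawn count by distance, and sparks enlarge with LOD distance.
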NiagaraSystem /DragonCartNarrative/Assets/VFX/ShardElectrify/NS_DragonCart_Shard_Electricity_Long
dependencies {
 usedDataInterfaces [&NiagaraDataInterfaceCurve, &NiagaraDataInterfaceAudioPlayer, &NiagaraDataInterfaceParticleRead]
 usedDynamicInputs [&MakeCustomLinearColorFromBool, &RandomBool]
 usedModules [
  &self:SystemSpawnScript.SystemScriptSource.SystemScriptGraph.NiagaraNodeAssignment_0.SetVariables_#1
  &SystemState
 ]
}
emitters {
 Lightning_01 {
  emitterSpawnScript [/SetVariables(Emitter.LightningWidth=1) {isSetParameters=true}]
  emitterUpdateScript [
   &EmitterState(
    Enable Distance Culling=true
    Max Distance Response=::Awaken
    MaxDistance=22000
    MinDistance=4000
    Module.bApplySpawnCountScaleByDistance=true
    Module.bUseMaxDistance=true
    Module.bUseMinDistance=false
    Scalability Mode=::NewEnumerator1
    Scale Spawn Count=true
   )
   &SpawnRate(SpawnRate=&Multiply_Float(A=$User.Intensity, B=1), Use Spawn Probability=true)
  ]
  particleSpawnScript [
   &InitializeParticle(
    Color=$System.MainColor
    Lifetime=&Lerp_Float(
     A=&RandomRangeFloat(Minimum=0.4, Maximum=0.5)
     B=&RandomRangeFloat(Minimum=0.25, Maximum=0.35)
     Alpha=$User.Intensity
    )
    Mesh Scale=&RandomRangeVector(Minimum=$User.Lighting_MinSize, Maximum=$User.Lightning_MaxSize)
    Mesh Scale Mode=::NewEnumerator4
    Sprite Rotation Mode=::NewEnumerator2
    Sprite Size Mode=::NewEnumerator1
    Uniform Sprite Size Max=1000
    Uniform Sprite Size Min=800
    UsePositionOffset=false
    Write Material Random=true
   )
   &ShapeLocation(
    Box Size=$User.SpawnSize
    MaskSpawnGroup=false
    Override Seed=false
    Shape Primitive=::NewEnumerator2
    Surface Only=false
    UseOwnerScale=true
   )
   /SetVariables(Particles.LightExponent=2, Particles.LightRadius=500) {isSetParameters=true}
   &InitialMeshOrientation(
    Mesh Orientation Mode=::NewEnumerator7
    Rotation=&RandomRangeVector(Minimum=$User.MeshOrientation_Min, Maximum=$User.MeshPrientation_Max)
    Use Rotation Vector=true
   )
  ]
  particleUpdateScript [
   &ParticleState()
   /SetVariables(Particles.LightningLight_Pos:NiagaraPosition=&AddVectorToPosition(Position=$Particles.Position, Vector=<600 0 0>)) {isSetParameters=true}
   /SetVariables(Particles.Flare_Color:LinearColor=&Multiply_LinearColorByFloat(
     Linear Color=$System.MainColor
     Float=&FloatFromCurve(
      FloatCurve=DataInterface() {propertyValues {Curve {keys [
          {interpMode=RCIM_Cubic, leaveTangent=36.0006, tangentMode=RCTM_Break, time=0.26, value=0}
          {interpMode=RCIM_Cubic, tangentMode=RCTM_Break, time=0.32, value=1}
         ]}}}
      Scale Curve=&Multiply_Float(A=1, B=$User.Intensity)
     )
    )) {isSetParameters=true}
   /SetVariables(Particles.LightColor=&Multiply_LinearColorByFloat(
     Linear Color=$System.MainColor
     Float=&FloatFromCurve(
      FloatCurve=DataInterface() {propertyValues {Curve {keys [<at 0.26 0>,<at 0.32 1>]}}}
      Scale Curve=&Multiply_Float(A=3, B=$User.Intensity)
     )
    )) {isSetParameters=true}
   /SetVariables(Particles.LightningGlow_Color:LinearColor=&Multiply_LinearColorByFloat(Linear Color=$System.MainColor, Float=&Multiply_Float(A=0.02, B=$User.Intensity))) {isSetParameters=true}
   &DynamicMaterialParameters(
    Index 0 Param 1=&FloatFromCurve(FloatCurve=DataInterface() {propertyValues {Curve {keys [<at 0 1>,<at 1 0.1>]}}})
    Index 0 Param 2=&Multiply_Float(A=0.05, B=1)
    Index 0 Param 3=0.03
    Index 0 Param 4=$Particles.MaterialRandom
    Index 1 Param 1=&FloatFromCurve(
     FloatCurve=DataInterface() {propertyValues {Curve {defaultValue=1, keys [{tangentMode=RCTM_User, time=0, value=0}, {tangentMode=RCTM_User, time=1, value=1}]}}}
     Scale Curve=4
    )
    Index 1 Param 3=0.2
    Index 1 Param 4=&FloatFromCurve(
     FloatCurve=DataInterface() {propertyValues {Curve {keys [<at 0 200>,<at 1 450>]}}}
     Scale Curve=&RandomRangeFloat(Minimum=0.75, Maximum=1)
    )
    Index 2 Param 1=400
    Index 3 Param 2=1
    Module.Index0Float0Write=true
    Module.Index0Float1Write=true
    Module.Index0Float2Write=true
    Module.Index0Float3Write=true
    Module.Index1Float0Write=true
    Module.Index1Float1Write=true
    Module.Index1Float2Write=true
    Module.Index1Float3Write=true
    Module.Index2Float0Write=true
    Module.Index2Float1Write=false
    Module.Index2Float2Write=false
    Module.Index2Float3Write=false
    Module.Index3Float0Write=false
    Module.Index3Float1Write=false
    Module.Index3Float2Write=false
    Module.Index3Float3Write=false
    Write Parameter Index 1=::NewEnumerator1
    Write Parameter Index 2=::NewEnumerator1
    Write Parameter Index 3=::NewEnumerator1
   ) {module=DynamicMaterialParameters001}
   /SetVariables(Particles.LightningGlow_Dynamic:Vector4f=&MakeVector4(
     X=&FloatFromCurve(FloatCurve=DataInterface() {propertyValues {Curve {keys [<at 0 1>,<at 1 0.1>]}}})
     Y=0.3
     Z=0.03
     W=$Particles.MaterialRandom
    )) {isSetParameters=true}
   &UpdateMeshOrientation(Rotation Rate=&RandomRangeFloat(Minimum=-0.3, Maximum=0.3))
   &CurlNoiseForce(Bias Noise=false, Noise Frequency=10, Noise Strength=600, Pan Noise=true, Pan Noise Field=<0 0 1>)
   &SolveForcesAndVelocity(Clamp Velocity=false, Limit Acceleration=false)
   &ScaleColor(Scale RGB=&VectorFromFloat(Value=$User.Intensity), ScaleA=true, ScaleRGB=true)
  ]
  rendererClasses [
   &NiagaraMeshRendererProperties
   &NiagaraSpriteRendererProperties
   &NiagaraLightRendererProperties
  ]
  renderers [
   {rendererClass=&NiagaraMeshRendererProperties, rendererIndex=0}
   {rendererClass=&NiagaraSpriteRendererProperties, rendererIndex=1}
   {rendererClass=&NiagaraLightRendererProperties, rendererIndex=2}
   {rendererClass=&NiagaraMeshRendererProperties, rendererIndex=3}
  ]
 }
 Sparks001 {
  emitterSpawnScript [&InitializeAudioPlayer(Audio Player=DataInterface() {propertyValues {
      Attenuation=null
      Concurrency=null
      MaxPlaysPerTick=10
      SoundToPlay=/DragonCartNarrative/Assets/Audio/OrbShockSequence/MSS_DragonCart_ShockSeq_Zaps
      bAllowLoopingOneShotSounds=false
      bLimitPlaysPerTick=true
      bOnlyActiveDuringGameplay=false
      bStopWhenComponentIsDestroyed=true
      ConfigurationUserParameter {parameter {name=None, typeDefHandle {registeredTypeIndex=1}, varData []}}
      ParameterNames []
     }})]
  emitterUpdateScript [
   &EmitterState(
    Enable Distance Culling=true
    Max Distance Response=::Awaken
    MaxDistance=12000
    MinDistance=4000
    Module.bApplySpawnCountScaleByDistance=true
    Module.bUseMaxDistance=true
    Module.bUseMinDistance=false
    Scalability Mode=::NewEnumerator1
    Scale Spawn Count=true
   )
   &SpawnParticlesFromOtherEmitter(
    Attribute Reader=DataInterface() {propertyValues {EmitterBinding {bindingMode=Other, emitterName=Lightning_01}}}
    Minimum Age Spawn Threshold=0.25
    Spawn Rate=&Lerp_Float(A=150, B=200, Alpha=$User.Intensity)
    Spawn Rate Per Particle Cap:NiagaraFloat=&Multiply_Float(A=1, B=&RandomRangeFloat(Minimum=100, Maximum=200))
    Use Maximum Age=false
    Use Minimum Age=false
   )
  ]
  particleSpawnScript [
   /SetVariables(Particles.Distribution:NiagaraFloat=&FloatFromCurve(
     FloatCurve=DataInterface() {propertyValues {Curve {keys [
         {interpMode=RCIM_Cubic, leaveTangent=0.4364, tangentMode=RCTM_Break, time=0, value=0}
         {arriveTangent=1.4038, interpMode=RCIM_Cubic, leaveTangent=1.4038, tangentMode=RCTM_User, time=0.6873, value=0.3824}
         <at 1 1 RCIM_Cubic>
        ]}}}
     CurveIndex=&RandomRangeFloat(Minimum=0, Maximum=1)
    )) {isSetParameters=true}
   &InitializeParticle(
    Color Mode=::NewEnumerator0
    Lifetime=&Lerp_Float(
     A=&RandomRangeFloat(Minimum=0.5, Maximum=0.7)
     B=&RandomRangeFloat(Minimum=0.6, Maximum=1.2)
     Alpha=$User.Intensity
    )
    Sprite Size=&Lerp_Vector2(
     A=&RandomRangeVector2D(Minimum=<9 11>, Maximum=<10 15>)
     B=&RandomRangeVector2D(Minimum=<11 12>, Maximum=<12 16>)
     Alpha=$User.Intensity
    )
    Sprite Size Mode=::NewEnumerator4
    UsePositionOffset=false
    Write Material Random=true
   )
   &SampleParticlesFromOtherEmitter(
    Applied RGB Scale:Vector3f=&VectorFromFloat(Value=&Multiply_Float(A=1, B=1))
    Applied Velocity Scale:NiagaraFloat=&RandomRangeFloat(Minimum=0.2, Maximum=0.4)
    Color Sampling=::NewEnumerator0
    Velocity Sampling=::NewEnumerator0
   )
   &ShapeLocation(MaskSpawnGroup=false, Override Seed=false, Sphere Radius=3, UseOwnerScale=true)
   &AddVelocity(
    Cone Angle=200
    Cone Axis=<0 0 1>
    Module.UseConeFalloff=true
    Module.UseVelDistribution=true
    Override Seed=false
    Rotation Coordinate Space=::World
    Velocity Mode=::NewEnumerator2
    Velocity Speed=&Lerp_Float(
     A=&RandomRangeFloat(Minimum=500, Maximum=900)
     B=&RandomRangeFloat(Minimum=1000, Maximum=1500)
     Alpha=$User.Intensity
    )
   )
   /Niagara/Modules/Audio/PlayAudio(
    Audio Playback Position:NiagaraPosition=&AddVectorToPosition(Position=$Particles.Position, Vector=<0 0 60>)
    Volume:NiagaraFloat=$User.ZapVolume
   )
  ]
  particleUpdateScript [
   &ParticleState()
   &DynamicMaterialParameters(
    Index 0 Param 1=&FloatFromCurve(FloatCurve=DataInterface() {propertyValues {Curve {keys [<at 0.1 0.5 RCIM_Cubic>,<at 1 4>]}}})
    Index 0 Param 2=5
    Index 0 Param 3=&FloatFromCurve(FloatCurve=DataInterface() {propertyValues {Curve {keys [<at 0 3 RCIM_Cubic>,<at 1 1>]}}}, Scale Curve=3)
    Index 0 Param 4=&FloatFromCurve(FloatCurve=DataInterface() {propertyValues {Curve {keys [{arriveTangent=-1.8572, interpMode=RCIM_Cubic, leaveTangent=-1.8572, tangentMode=RCTM_User, time=0, value=0.25}, <at 0.5 0>]}}})
    Module.Index0Float0Write=true
    Module.Index0Float1Write=false
    Module.Index0Float2Write=true
    Module.Index0Float3Write=true
   )
   &GravityForce(Gravity=<0 0 -2200>)
   &Drag(Drag=0.7, Use Linear Drag=true, Use Rotational Drag=false)
   &SolveForcesAndVelocity(Clamp Velocity=false, Limit Acceleration=false)
   &ScaleSpriteSize(
    Uniform Curve Scale=&FloatFromCurve(
     FloatCurve=DataInterface() {propertyValues {Curve {keys [
         {arriveTangent=0.001, interpMode=RCIM_Cubic, leaveTangent=0.001, tangentMode=RCTM_User, time=1000, value=0.5}
         {arriveTangentWeight=166.6667, interpMode=RCIM_Cubic, leaveTangentWeight=400.0001, tangentMode=RCTM_User, time=2000, value=1}
         <at 30000 5>
        ]}}}
     CurveIndex=$Engine.Owner.LODDistance
    )
    Uniform Curve Sprite Scale=DataInterface() {propertyValues {Curve {keys [<at 0.5 1 RCIM_Cubic>, {arriveTangent=-5.8778, interpMode=RCIM_Cubic, leaveTangent=-5.8778, tangentMode=RCTM_User, time=1, value=0}]}}}
   )
   &ScaleSpriteSizeBySpeed(
    Max Scale Factor=<0.8 10>
    Sample Scale Factor By Curve=true
    Scale Factor Curve=DataInterface() {propertyValues {Curve {keys [{arriveTangent=0.5993, interpMode=RCIM_Cubic, time=0, value=0}, {interpMode=RCIM_Cubic, leaveTangent=2.2363, time=1, value=1}]}}}
    Velocity Threshold=4100
   )
   &ScaleColor(
    Scale RGB=&VectorFromFloat(Value=&Lerp_Float(A=1, B=3, Alpha=0))
    ScaleA=false
    ScaleRGB=true
   )
  ]
  rendererClasses [&NiagaraSpriteRendererProperties]
  renderers [{rendererClass=&NiagaraSpriteRendererProperties}]
 }
}
summary {
 systemName=NS_DragonCart_Shard_Electricity_Long
 emitters [
  {
   emitterName=Lightning_01
   rendererClasses [
    &NiagaraMeshRendererProperties
    &NiagaraSpriteRendererProperties
    &NiagaraLightRendererProperties
   ]
  }
  {emitterName=Sparks001, rendererClasses [&NiagaraSpriteRendererProperties]}
 ]
 userVariables [
  User.Intensity {type=NiagaraFloat, defaultValue=1, description=""}
  User.Lighting_MinSize {type=Vector3f, description="", defaultValue {struct=Vector3f, value=<1.3 1 1>}}
  User.Lightning_MaxSize {type=Vector3f, description="", defaultValue {struct=Vector3f, value=<1.4 1.4 1.4>}}
  User.MeshOrientation_Min {type=Vector3f, description="", defaultValue {struct=Vector3f, value=<-1 -0.02 -0.03>}}
  User.MeshPrientation_Max {type=Vector3f, description="", defaultValue {struct=Vector3f, value=<1 0.02 0.03>}}
  User.SpawnSize {type=Vector3f, description="", defaultValue {struct=Vector3f, value=<0 500 0>}}
  User.ZapVolume {type=NiagaraFloat, defaultValue=1, description=""}
 ]
}
