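# Niagara system for the electrified DragonCart shard effect, drawing flickering blue lightning arcs with mesh and sprite renderers.
# ElectricArcs emitters place particles using destination vectors, with Buildup driving spawn rate, lifetime and color, plus dynamic material parameters and distance culling.
# MedLightning samples the DragonCart fragment mesh, stretching bolts from near the owner and adding flares and lights.
# Among the emitters fully shown, only ElectricArcs002, which scatters arcs in a sphere, is enabled.
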NiagaraSystem /DragonCartNarrative/Assets/VFX/ShardElectrify/NS_NDC_DragonCart_Shard_Eletrify
dependencies {
 usedDataInterfaces [
  /Script/Niagara.NiagaraDataInterfaceEmitterProperties
  &NiagaraDataInterfaceCurve
  /Script/Niagara.NiagaraDataInterfaceArrayBool
  &NiagaraDataInterfaceArrayFloat
  &NiagaraDataInterfaceMeshRendererInfo
  &NiagaraDataInterfaceStaticMesh
  &NiagaraDataInterfaceCamera
  &NiagaraDataInterfaceVectorCurve
  &NiagaraDataInterfaceArrayFloat3
 ]
 usedDynamicInputs [&SelectVectorFromArray]
 usedModules [
  &SystemState
  &self:SystemSpawnScript.SystemScriptSource.SystemScriptGraph.NiagaraNodeAssignment_1.SetVariables_#1
 ]
}
emitters {
 ElectricArcs {
  bEnabled=false
  emitterSpawnScript [&self:Set_Fixed_Bounds(
    Emitter Properties=DataInterface() {propertyValues {EmitterBinding {bindingMode=Self, emitterName=None}}}
    Length:NiagaraFloat=&VectorLength(Vector=$User.LocalDestination)
   ) {enabled=false, module=Set_Fixed_Bounds}]
  emitterUpdateScript [
   &EmitterState(Enable Distance Culling=true, MaxDistance=7000, Module.bUseMaxDistance=true, Module.bUseMinDistance=false, Scalability Mode=::NewEnumerator1)
   &SpawnRate(
    Spawn Probability=0.5
    SpawnRate=&Lerp_Float(A=&RandomRangeFloat(Minimum=1, Maximum=2), B=6, Alpha=$User.Buildup)
    Use Spawn Probability=true
   )
  ]
  particleSpawnScript [
   &InitializeParticle(
    Color=&RandomRangeLinearColor(Minimum=<rgba 0.0986 0.6632 1 1>, Maximum=<rgba 0.3463 1.473 2 1>)
    Lifetime=&FloatFromCurve(
     FloatCurve=DataInterface() {propertyValues {Curve {keys [<at 0 1 RCIM_Cubic>, {arriveTangent=-1.1908, interpMode=RCIM_Cubic, leaveTangent=-1.1908, tangentMode=RCTM_User, time=0.9, value=0.3}]}}}
     CurveIndex=$System.Buildup
     Scale Curve=&UniformRangedFloat(Minimum=0.6, Maximum=1.5)
    )
    Mesh Scale Max=<2 5 3.5>
    Mesh Scale Min=<1 3 2.5>
    Mesh Scale Mode=::NewEnumerator2
    Sprite Rotation Mode=::NewEnumerator2
    Sprite Size Mode=::NewEnumerator3
    Uniform Sprite Size=500
    UsePositionOffset=false
    Write Material Random=true
   ) {module=InitializeParticle001}
   /SetVariables(Particles.Position=&AddVectorToPosition(Position=$Engine.Emitter.SimulationPosition, Vector=$User.LocalDestination)) {isSetParameters=true, module=SetVariables_#2}
   &InitialMeshOrientation(
    Mesh Axis To Orient=<-1 0 0>
    Mesh Orientation Mode=::NewEnumerator2
    Orientation Vector=$Particles.ShapeLocation.ShapeNormal
    Rotation=&RandomRangeVector(Minimum=<0 0 0>, Maximum=<1 1 1>)
    Use Rotation Vector=true
   ) {module=InitialMeshOrientation001}
   &CameraOffset(Camera Offset Amount=50)
   /SetVariables() {isSetParameters=true, module=SetVariables_#3}
  ]
  particleUpdateScript [
   &ParticleState()
   &ScaleMeshSize(Scale Factor=&MakeVector(
     X=&Multiply_Float(
      A=&VectorLength(Vector=$User.LocalDestination)
      B=&RandomRangeFloat(Minimum=0.001, Maximum=0.004)
     )
     Y=5
     Z=5
    ))
   &ScaleColor(
    Scale Alpha=&self:Flicker_01(
     FlickerAmount:NiagaraFloat=&FloatFromCurve(
      FloatCurve=DataInterface() {propertyValues {Curve {keys [
          {arriveTangent=9.8056, interpMode=RCIM_Cubic, time=0.15, value=1}
          {interpMode=RCIM_Cubic, tangentMode=RCTM_User, time=0.5, value=0.2}
          {arriveTangent=1.9125, interpMode=RCIM_Cubic, leaveTangent=1.9125, tangentMode=RCTM_User, time=1, value=1}
         ]}}}
      Scale Curve=2
     )
     FlickerInterval:NiagaraFloat=0.05
    )
    Scale RGB=&MakeVectorFromLinearColorRGB(Color=&LerpLinearColors(StartColor=<rgba 0.012 0.1273 0.5 1>, EndColor=<rgba 0.8276 0.9281 1 1>, LerpFactor=$System.Buildup))
    ScaleA=true
    ScaleRGB=true
   )
   &DynamicMaterialParameters(
    Index 0 Param 1=&RandomRangeFloat(Minimum=15, Maximum=20)
    Index 0 Param 2=&RandomRangeFloat(Minimum=8, Maximum=12)
    Index 0 Param 3=&RandomRangeFloat(Minimum=2, Maximum=3)
    Index 0 Param 4=&UniformRangedFloat(Minimum=3, Maximum=1)
    Module.Index0Float0Write=true
    Module.Index0Float1Write=true
    Module.Index0Float2Write=true
    Module.Index0Float3Write=true
   )
   /SetVariables(Particles.DynamicParamOverride:Vector4f=&MakeFloat4FromVectorAndFloat(
     "Vector (RGB)"=&AddVectorToPosition(Position=$Engine.Owner.Position, Vector=&MakeVector(X=0, Y=0, Z=0))
     "Float (Alpha)"=&FloatFromCurve(FloatCurve=DataInterface() {propertyValues {Curve {keys [<at 0 0 RCIM_Cubic>,<at 1 1>]}}})
    )) {isSetParameters=true, module=SetVariables_#4}
  ]
  rendererClasses [&NiagaraMeshRendererProperties, &NiagaraSpriteRendererProperties]
  renderers [
   {rendererClass=&NiagaraMeshRendererProperties, rendererIndex=0}
   {rendererClass=&NiagaraSpriteRendererProperties, rendererIndex=1}
  ]
 }
 ElectricArcs001 {
  bEnabled=false
  emitterSpawnScript [&self:Set_Fixed_Bounds(
    Emitter Properties=DataInterface() {propertyValues {EmitterBinding {bindingMode=Self, emitterName=None}}}
    Length:NiagaraFloat=&VectorLength(Vector=$User.LocalDestination)
   ) {enabled=false, module=Set_Fixed_Bounds}]
  emitterUpdateScript [
   &EmitterState(Enable Distance Culling=true, MaxDistance=7000, Module.bUseMaxDistance=true, Module.bUseMinDistance=false, Scalability Mode=::NewEnumerator1)
   &SpawnRate(
    Spawn Probability=0.5
    SpawnRate=&Lerp_Float(A=&RandomRangeFloat(Minimum=1, Maximum=1), B=6, Alpha=$User.Buildup)
    Use Spawn Probability=true
   )
  ]
  particleSpawnScript [
   &InitializeParticle(
    Color=&RandomRangeLinearColor(Minimum=<rgba 0 0.6264 1 1>, Maximum=<rgba 0.3463 1.473 2 1>)
    Lifetime=&FloatFromCurve(
     FloatCurve=DataInterface() {propertyValues {Curve {keys [<at 0 1 RCIM_Cubic>, {arriveTangent=-1.1908, interpMode=RCIM_Cubic, leaveTangent=-1.1908, tangentMode=RCTM_User, time=0.9, value=0.3}]}}}
     CurveIndex=$System.Buildup
     Scale Curve=&UniformRangedFloat(Minimum=0.3, Maximum=0.8)
    )
    Mesh Scale Max=<2 5 3.5>
    Mesh Scale Min=<1 3 2.5>
    Mesh Scale Mode=::NewEnumerator2
    Sprite Rotation Mode=::NewEnumerator2
    Sprite Size Mode=::NewEnumerator3
    UsePositionOffset=false
    Write Material Random=true
   ) {module=InitializeParticle001}
   /SetVariables(Particles.Position=&AddVectorToPosition(Position=$Engine.Emitter.SimulationPosition, Vector=$System.MultipleDestination)) {isSetParameters=true, module=SetVariables_#2}
   &InitialMeshOrientation(
    Mesh Axis To Orient=<-1 0 0>
    Mesh Orientation Mode=::NewEnumerator2
    Orientation Vector=$Particles.ShapeLocation.ShapeNormal
    Rotation=&RandomRangeVector(Minimum=<0 0 0>, Maximum=<1 1 1>)
    Use Rotation Vector=true
   ) {module=InitialMeshOrientation001}
   /SetVariables(Particles.SpriteCameraOffset=100, Particles.SpriteSize=&Multiply_Vector2D_ByFloat(Float=700, Vector2D=<1 1>)) {isSetParameters=true, module=SetVariables_#3}
  ]
  particleUpdateScript [
   &ParticleState()
   &ScaleMeshSize(Scale Factor=&MakeVector(
     X=&Multiply_Float(
      A=&VectorLength(Vector=$System.MultipleDestination)
      B=&RandomRangeFloat(Minimum=0.002, Maximum=0.009)
     )
     Y=5
     Z=5
    ))
   &ScaleColor(
    Scale Alpha=&self:Flicker_01(
     FlickerAmount:NiagaraFloat=&FloatFromCurve(
      FloatCurve=DataInterface() {propertyValues {Curve {keys [
          {arriveTangent=9.8056, interpMode=RCIM_Cubic, time=0.15, value=1}
          {interpMode=RCIM_Cubic, tangentMode=RCTM_User, time=0.5, value=0.2}
          {arriveTangent=1.9125, interpMode=RCIM_Cubic, leaveTangent=1.9125, tangentMode=RCTM_User, time=1, value=1}
         ]}}}
      Scale Curve=2
     )
     FlickerInterval:NiagaraFloat=0.05
    )
    Scale RGB=&MakeVectorFromLinearColorRGB(Color=&LerpLinearColors(StartColor=<rgba 0.0241 0.2546 1 1>, EndColor=<rgba 0.8276 0.9281 1 1>, LerpFactor=$System.Buildup))
    ScaleA=true
    ScaleRGB=true
   )
   &DynamicMaterialParameters(
    Index 0 Param 1=&RandomRangeFloat(Minimum=10, Maximum=15)
    Index 0 Param 2=&RandomRangeFloat(Minimum=8, Maximum=12)
    Index 0 Param 3=&RandomRangeFloat(Minimum=2, Maximum=3)
    Index 0 Param 4=&UniformRangedFloat(Minimum=1, Maximum=1.5)
    Module.Index0Float0Write=true
    Module.Index0Float1Write=true
    Module.Index0Float2Write=true
    Module.Index0Float3Write=true
   )
   /SetVariables(Particles.DynamicParamOverride:Vector4f=&MakeFloat4FromVectorAndFloat(
     "Vector (RGB)"=&AddVectorToPosition(Position=$Engine.Owner.Position, Vector=&MakeVector(X=0, Y=0, Z=0))
     "Float (Alpha)"=&FloatFromCurve(FloatCurve=DataInterface() {propertyValues {Curve {keys [<at 0 0 RCIM_Cubic>,<at 1 1>]}}})
    )) {isSetParameters=true, module=SetVariables_#4}
  ]
  rendererClasses [&NiagaraMeshRendererProperties, &NiagaraSpriteRendererProperties]
  renderers [
   {rendererClass=&NiagaraMeshRendererProperties, rendererIndex=0}
   {rendererClass=&NiagaraSpriteRendererProperties, rendererIndex=1}
   {rendererClass=&NiagaraSpriteRendererProperties, rendererIndex=2}
  ]
 }
 ElectricArcs002 {
  emitterSpawnScript [&self:Set_Fixed_Bounds(
    Emitter Properties=DataInterface() {propertyValues {EmitterBinding {bindingMode=Self, emitterName=None}}}
    Length:NiagaraFloat=&VectorLength(Vector=Linked() {linkedVariable {name=User.LocalDestination, type=NiagaraPosition}})
   ) {enabled=false, module=Set_Fixed_Bounds}]
  emitterUpdateScript [
   &EmitterState(Enable Distance Culling=true, MaxDistance=7000, Module.bUseMaxDistance=true, Module.bUseMinDistance=false, Scalability Mode=::NewEnumerator1)
   &SpawnRate(
    Spawn Probability=0.5
    SpawnRate=&Lerp_Float(A=&RandomRangeFloat(Minimum=3, Maximum=5), B=6, Alpha=$User.Buildup)
    Use Spawn Probability=true
   )
  ]
  particleSpawnScript [
   &InitializeParticle(
    Color=&RandomRangeLinearColor(Minimum=<rgba 0 0.6264 1 1>, Maximum=<rgba 0.3463 1.473 2 1>)
    Lifetime=&FloatFromCurve(
     FloatCurve=DataInterface() {propertyValues {Curve {keys [<at 0 1 RCIM_Cubic>, {arriveTangent=-1.1908, interpMode=RCIM_Cubic, leaveTangent=-1.1908, tangentMode=RCTM_User, time=0.9, value=0.3}]}}}
     CurveIndex=$System.Buildup
     Scale Curve=&UniformRangedFloat(Minimum=0.3, Maximum=0.8)
    )
    Mesh Scale Max=<2 5 3.5>
    Mesh Scale Min=<1 3 2.5>
    Mesh Scale Mode=::NewEnumerator2
    Sprite Rotation Mode=::NewEnumerator2
    Sprite Size Mode=::NewEnumerator3
    UsePositionOffset=false
    Write Material Random=true
   ) {module=InitializeParticle001}
   /SetVariables(Particles.Position=&AddVectorToPosition(Position=$Engine.Emitter.SimulationPosition, Vector=Linked() {linkedVariable {name=User.LocalDestination, type=NiagaraPosition}})) {isSetParameters=true, module=SetVariables_#2}
   &InitialMeshOrientation(
    Mesh Axis To Orient=<-1 0 0>
    Mesh Orientation Mode=::NewEnumerator2
    Orientation Vector=$Particles.ShapeLocation.ShapeNormal
    Rotation=&RandomRangeVector(Minimum=<0 0 0>, Maximum=<1 1 1>)
    Use Rotation Vector=true
   ) {module=InitialMeshOrientation001}
   /SetVariables(Particles.SpriteCameraOffset=100, Particles.SpriteSize=&Multiply_Vector2D_ByFloat(Float=700, Vector2D=<1 1>)) {isSetParameters=true, module=SetVariables_#3}
  ]
  particleUpdateScript [
   &ParticleState()
   &ShapeLocation(MaskSpawnGroup=false, Override Seed=false, Shape Origin=$User.LocalDestination, Sphere Radius=1000, UseOwnerScale=true)
   &ScaleMeshSize(Scale Factor=&MakeVector(
     X=&Multiply_Float(
      A=&VectorLength(Vector=Linked() {linkedVariable {name=User.LocalDestination, type=NiagaraPosition}})
      B=&RandomRangeFloat(Minimum=0.002, Maximum=0.009)
     )
     Y=5
     Z=5
    ))
   &ScaleColor(
    Scale Alpha=&self:Flicker_01(
     FlickerAmount:NiagaraFloat=&FloatFromCurve(
      FloatCurve=DataInterface() {propertyValues {Curve {keys [
          {arriveTangent=9.8056, interpMode=RCIM_Cubic, time=0.15, value=1}
          {interpMode=RCIM_Cubic, tangentMode=RCTM_User, time=0.5, value=0.2}
          {arriveTangent=1.9125, interpMode=RCIM_Cubic, leaveTangent=1.9125, tangentMode=RCTM_User, time=1, value=1}
         ]}}}
      Scale Curve=2
     )
     FlickerInterval:NiagaraFloat=0.05
    )
    Scale RGB=&MakeVectorFromLinearColorRGB(Color=&LerpLinearColors(StartColor=<rgba 0.0241 0.2546 1 1>, EndColor=<rgba 0.8276 0.9281 1 1>, LerpFactor=$System.Buildup))
    ScaleA=true
    ScaleRGB=true
   )
   &DynamicMaterialParameters(
    Index 0 Param 1=&RandomRangeFloat(Minimum=10, Maximum=15)
    Index 0 Param 2=&RandomRangeFloat(Minimum=8, Maximum=12)
    Index 0 Param 3=&RandomRangeFloat(Minimum=2, Maximum=3)
    Index 0 Param 4=&UniformRangedFloat(Minimum=1, Maximum=1.5)
    Module.Index0Float0Write=true
    Module.Index0Float1Write=true
    Module.Index0Float2Write=true
    Module.Index0Float3Write=true
   )
   /SetVariables(Particles.DynamicParamOverride:Vector4f=&MakeFloat4FromVectorAndFloat(
     "Vector (RGB)"=&AddVectorToPosition(Position=$Engine.Owner.Position, Vector=&MakeVector(X=0, Y=0, Z=0))
     "Float (Alpha)"=&FloatFromCurve(FloatCurve=DataInterface() {propertyValues {Curve {keys [<at 0 0 RCIM_Cubic>,<at 1 1>]}}})
    )) {isSetParameters=true, module=SetVariables_#4}
   /SetVariables() {isSetParameters=true, module=SetVariables_#5}
  ]
  rendererClasses [&NiagaraMeshRendererProperties, &NiagaraSpriteRendererProperties]
  renderers [
   {rendererClass=&NiagaraMeshRendererProperties, rendererIndex=0}
   {rendererClass=&NiagaraSpriteRendererProperties, rendererIndex=1}
   {rendererClass=&NiagaraSpriteRendererProperties, rendererIndex=2}
  ]
 }
 ElectricArcs003 {
  bEnabled=false
  emitterSpawnScript [&self:Set_Fixed_Bounds(
    Emitter Properties=DataInterface() {propertyValues {EmitterBinding {bindingMode=Self, emitterName=None}}}
    Length:NiagaraFloat=&VectorLength(Vector=$User.LocalDestination)
   ) {enabled=false, module=Set_Fixed_Bounds}]
  emitterUpdateScript [
   &EmitterState(Enable Distance Culling=true, MaxDistance=7000, Module.bUseMaxDistance=true, Module.bUseMinDistance=false, Scalability Mode=::NewEnumerator1)
   &SpawnRate(
    Spawn Probability=0.5
    SpawnRate=&Lerp_Float(A=&RandomRangeFloat(Minimum=1, Maximum=2), B=6, Alpha=$User.Buildup)
    Use Spawn Probability=true
   )
  ]
  particleSpawnScript [
   &InitializeParticle(
    Color=&RandomRangeLinearColor(Minimum=<rgba 0.0986 0.6632 1 1>, Maximum=<rgba 0.3463 1.473 2 1>)
    Lifetime=&FloatFromCurve(
     FloatCurve=DataInterface() {propertyValues {Curve {keys [<at 0 1 RCIM_Cubic>, {arriveTangent=-1.1908, interpMode=RCIM_Cubic, leaveTangent=-1.1908, tangentMode=RCTM_User, time=0.9, value=0.3}]}}}
     CurveIndex=$System.Buildup
     Scale Curve=&UniformRangedFloat(Minimum=0.6, Maximum=1.5)
    )
    Mesh Scale=&MakeVector(
     X=1
     Y=&RandomRangeFloat(Minimum=2, Maximum=4)
     Z=&RandomRangeFloat(Minimum=2, Maximum=4)
    )
    Mesh Scale Mode=::NewEnumerator4
    Sprite Rotation Mode=::NewEnumerator2
    Sprite Size Mode=::NewEnumerator3
    Uniform Sprite Size=500
    UsePositionOffset=false
    Write Material Random=true
   ) {module=InitializeParticle001}
   /SetVariables(Particles.Position=&AddVectorToPosition(Position=$Engine.Emitter.SimulationPosition, Vector=$User.LocalDestination)) {isSetParameters=true, module=SetVariables_#2}
   &InitialMeshOrientation(
    Mesh Axis To Orient=<-1 0 0>
    Mesh Orientation Mode=::NewEnumerator2
    Orientation Vector=$Particles.ShapeLocation.ShapeNormal
    Rotation=&RandomRangeVector(Minimum=<0 0 0>, Maximum=<1 1 1>)
    Use Rotation Vector=true
   ) {module=InitialMeshOrientation001}
   &CameraOffset(Camera Offset Amount=50)
   /SetVariables() {isSetParameters=true, module=SetVariables_#3}
  ]
  particleUpdateScript [
   &ParticleState()
   &ScaleMeshSize(Scale Factor=&VectorFromCurve(
     VectorCurve=DataInterface() {propertyValues {
       XCurve {keys [
         {interpMode=RCIM_Cubic, leaveTangentWeight=0.1333, tangentMode=RCTM_User, time=0.6, value=1}
         {arriveTangent=-2.2924, tangentMode=RCTM_User, time=1, value=0.7}
        ]}
       YCurve {keys [
         {arriveTangent=3.3015, interpMode=RCIM_Cubic, leaveTangent=3.3015, tangentMode=RCTM_User, time=0, value=0.25}
         {arriveTangentWeight=0.1954, interpMode=RCIM_Cubic, leaveTangentWeight=0.1466, tangentMode=RCTM_User, time=0.4, value=1}
        ]}
       ZCurve {keys [
         {arriveTangent=3.3015, interpMode=RCIM_Cubic, leaveTangent=3.3015, tangentMode=RCTM_User, time=0, value=0.25}
         {arriveTangentWeight=0.1954, interpMode=RCIM_Cubic, leaveTangentWeight=0.1466, tangentMode=RCTM_User, time=0.4, value=1}
        ]}
      }}
     Scale Curve=&MakeVector(
      X=&Divide_Float(A=&DistanceBetweenPositions(A=$Particles.Position, B=$User.LocalDestination), B=1000)
      Y=1
      Z=1
     )
    )) {enabled=false, module=ScaleMeshSize001}
   &ScaleMeshSize(Scale Factor=&MakeVector(
     X=&Multiply_Float(A=&VectorLength(Vector=$User.LocalDestination), B=0)
     Y=0.1
     Z=0.1
    ))
   &ScaleColor(
    Scale Alpha=&self:Flicker_01(
     FlickerAmount:NiagaraFloat=&FloatFromCurve(
      FloatCurve=DataInterface() {propertyValues {Curve {keys [
          {arriveTangent=9.8056, interpMode=RCIM_Cubic, time=0.15, value=1}
          {interpMode=RCIM_Cubic, tangentMode=RCTM_User, time=0.5, value=0.2}
          {arriveTangent=1.9125, interpMode=RCIM_Cubic, leaveTangent=1.9125, tangentMode=RCTM_User, time=1, value=1}
         ]}}}
      Scale Curve=2
     )
     FlickerInterval:NiagaraFloat=0.05
    )
    Scale RGB=&MakeVectorFromLinearColorRGB(Color=&LerpLinearColors(StartColor=<rgba 0.012 0.1273 0.5 1>, EndColor=<rgba 0.8276 0.9281 1 1>, LerpFactor=$System.Buildup))
    ScaleA=true
    ScaleRGB=true
   )
   &DynamicMaterialParameters(
    Index 0 Param 1=&FloatFromCurve(FloatCurve=DataInterface() {propertyValues {Curve {
        defaultValue=0
        keys [
         {arriveTangent=-0.8644, interpMode=RCIM_Cubic, leaveTangent=-0.8644, tangentMode=RCTM_User, time=0, value=1}
         {arriveTangent=-3.8427, tangentMode=RCTM_User, time=0.35, value=0.5}
         <at 0.65 0.5>
         {tangentMode=RCTM_User, time=1, value=0}
        ]
       }}})
    Index 0 Param 2=2
    Index 0 Param 3=1
    Index 1 Param 1=&FloatFromCurve(FloatCurve=DataInterface() {propertyValues {Curve {defaultValue=1, keys [{tangentMode=RCTM_User, time=0, value=0}, {tangentMode=RCTM_User, time=1, value=1}]}}})
    Index 1 Param 2=25
    Index 1 Param 3=0.5
    Index 1 Param 4=1
    Index 2 Param 1=400
    Index 3 Param 2=1
    Module.Index0Float0Write=true
    Module.Index0Float1Write=true
    Module.Index0Float2Write=true
    Module.Index0Float3Write=true
    Module.Index1Float0Write=true
    Module.Index1Float1Write=true
    Module.Index1Float2Write=true
    Module.Index1Float3Write=true
    Module.Index2Float0Write=true
    Module.Index2Float1Write=false
    Module.Index2Float2Write=false
    Module.Index2Float3Write=false
    Module.Index3Float0Write=false
    Module.Index3Float1Write=false
    Module.Index3Float2Write=false
    Module.Index3Float3Write=false
    Write Parameter Index 1=::NewEnumerator1
    Write Parameter Index 2=::NewEnumerator1
    Write Parameter Index 3=::NewEnumerator1
   )
  ]
  rendererClasses [&NiagaraMeshRendererProperties, &NiagaraSpriteRendererProperties]
  renderers [
   {rendererClass=&NiagaraMeshRendererProperties, rendererIndex=0}
   {rendererClass=&NiagaraSpriteRendererProperties, rendererIndex=1}
  ]
 }
 ElectricArcs004 {
  bEnabled=false
  emitterSpawnScript [&self:Set_Fixed_Bounds(
    Emitter Properties=DataInterface() {propertyValues {EmitterBinding {bindingMode=Self, emitterName=None}}}
    Length:NiagaraFloat=&VectorLength(Vector=Linked() {linkedVariable {name=User.LocalDestination, type=NiagaraPosition}})
   ) {enabled=false, module=Set_Fixed_Bounds}]
  emitterUpdateScript [
   &EmitterState(Enable Distance Culling=true, MaxDistance=7000, Module.bUseMaxDistance=true, Module.bUseMinDistance=false, Scalability Mode=::NewEnumerator1)
   &SpawnRate(
    Spawn Probability=0.5
    SpawnRate=&Lerp_Float(A=&RandomRangeFloat(Minimum=3, Maximum=5), B=6, Alpha=$User.Buildup)
    Use Spawn Probability=true
   )
  ]
  particleSpawnScript [
   &InitializeParticle(
    Color=&RandomRangeLinearColor(Minimum=<rgba 0 0.6264 1 1>, Maximum=<rgba 0.3463 1.473 2 1>)
    Lifetime=&FloatFromCurve(
     FloatCurve=DataInterface() {propertyValues {Curve {keys [<at 0 1 RCIM_Cubic>, {arriveTangent=-1.1908, interpMode=RCIM_Cubic, leaveTangent=-1.1908, tangentMode=RCTM_User, time=0.9, value=0.3}]}}}
     CurveIndex=$System.Buildup
     Scale Curve=&UniformRangedFloat(Minimum=0.3, Maximum=0.8)
    )
    Mesh Scale Max=<2 5 3.5>
    Mesh Scale Min=<1 3 2.5>
    Mesh Scale Mode=::NewEnumerator2
    Sprite Rotation Mode=::NewEnumerator2
    Sprite Size Mode=::NewEnumerator3
    UsePositionOffset=false
    Write Material Random=true
   ) {module=InitializeParticle001}
   /SetVariables(Particles.Position=&AddVectorToPosition(Position=$Engine.Emitter.SimulationPosition, Vector=Linked() {linkedVariable {name=User.LocalDestination, type=NiagaraPosition}})) {isSetParameters=true, module=SetVariables_#2}
   &InitialMeshOrientation(
    Mesh Axis To Orient=<-1 0 0>
    Mesh Orientation Mode=::NewEnumerator2
    Orientation Vector=$Particles.ShapeLocation.ShapeNormal
    Rotation=&RandomRangeVector(Minimum=<0 0 0>, Maximum=<1 1 1>)
    Use Rotation Vector=true
   ) {module=InitialMeshOrientation001}
   /SetVariables(Particles.SpriteCameraOffset=100, Particles.SpriteSize=&Multiply_Vector2D_ByFloat(Float=700, Vector2D=<1 1>)) {isSetParameters=true, module=SetVariables_#3}
  ]
  particleUpdateScript [
   &ParticleState()
   &ShapeLocation(MaskSpawnGroup=false, Override Seed=false, Shape Origin=$User.LocalDestination, Sphere Radius=1000, UseOwnerScale=true)
   &ScaleMeshSize(Scale Factor=&MakeVector(
     X=&VectorLength(Vector=Linked() {linkedVariable {name=User.LocalDestination, type=NiagaraPosition}})
     Y=1
     Z=1
    ))
   &ScaleColor(
    Scale Alpha=&self:Flicker_01(
     FlickerAmount:NiagaraFloat=&FloatFromCurve(
      FloatCurve=DataInterface() {propertyValues {Curve {keys [
          {arriveTangent=9.8056, interpMode=RCIM_Cubic, time=0.15, value=1}
          {interpMode=RCIM_Cubic, tangentMode=RCTM_User, time=0.5, value=0.2}
          {arriveTangent=1.9125, interpMode=RCIM_Cubic, leaveTangent=1.9125, tangentMode=RCTM_User, time=1, value=1}
         ]}}}
      Scale Curve=2
     )
     FlickerInterval:NiagaraFloat=0.05
    )
    Scale RGB=&MakeVectorFromLinearColorRGB(Color=&LerpLinearColors(StartColor=<rgba 0.0241 0.2546 1 1>, EndColor=<rgba 0.8276 0.9281 1 1>, LerpFactor=$System.Buildup))
    ScaleA=true
    ScaleRGB=true
   )
   &DynamicMaterialParameters(
    Index 0 Param 1=&FloatFromCurve(FloatCurve=DataInterface() {propertyValues {Curve {defaultValue=0, keys [{tangentMode=RCTM_User, time=0, value=1}, {tangentMode=RCTM_User, time=1, value=0}]}}})
    Index 0 Param 2=10
    Index 0 Param 3=&RandomRangeFloat(Minimum=2, Maximum=3)
    Index 0 Param 4=&UniformRangedFloat(Minimum=1, Maximum=1.5)
    Index 1 Param 1=2
    Index 1 Param 2=1
    Index 1 Param 3=1
    Module.Index0Float0Write=true
    Module.Index0Float1Write=true
    Module.Index0Float2Write=true
    Module.Index0Float3Write=true
    Module.Index1Float0Write=true
    Module.Index1Float1Write=true
    Module.Index1Float2Write=true
    Module.Index1Float3Write=true
    Write Parameter Index 1=::NewEnumerator1
   )
   /SetVariables(Particles.DynamicParamOverride:Vector4f=&MakeFloat4FromVectorAndFloat(
     "Vector (RGB)"=&AddVectorToPosition(Position=$Engine.Owner.Position, Vector=&MakeVector(X=0, Y=0, Z=0))
     "Float (Alpha)"=&FloatFromCurve(FloatCurve=DataInterface() {propertyValues {Curve {keys [<at 0 0 RCIM_Cubic>,<at 1 1>]}}})
    )) {enabled=false, isSetParameters=true, module=SetVariables_#4}
   /SetVariables() {enabled=false, isSetParameters=true, module=SetVariables_#5}
  ]
  rendererClasses [&NiagaraMeshRendererProperties, &NiagaraSpriteRendererProperties]
  renderers [
   {rendererClass=&NiagaraMeshRendererProperties, rendererIndex=0}
   {rendererClass=&NiagaraSpriteRendererProperties, rendererIndex=1}
   {rendererClass=&NiagaraSpriteRendererProperties, rendererIndex=2}
  ]
 }
 MedLightning {
  bEnabled=false
  emitterSpawnScript [/SetVariables(Emitter.LightningColor:LinearColor=&MakeCustomLinearColorFromBool(Bool=$User.IsRed, True Color=<rgba 0.0508 0.234 1 1>, False Color=<rgba 0.0479 0.1435 1 1>)) {isSetParameters=true, module=SetVariables_#6}]
  emitterUpdateScript [
   &EmitterState()
   &SpawnBurst_Instantaneous(Spawn Count=8, Use Spawn Probability=false, UseLoopCountLimit=true)
   &SpawnRate(SpawnRate=14, Use Spawn Probability=false)
  ]
  particleSpawnScript [
   &PartitionParticles(Number Of Partitions=14)
   /SetVariables(Particles.Mirror:NiagaraFloat=&MakeCustomFloatFromBool(
     Bool=/Niagara/DynamicInputs/Arrays/SelectBoolFromArray(
      Array Sampling Mode:ENiagara_LifetimeMode=::NewEnumerator0
      Direct Array Index:NiagaraInt32=$Particles.PartitionParticles.Partition
      Direct Array Mode:ENiagara_WrapClamp=::NewEnumerator1
      Evaluation Type:ENiagaraRandomnessEvaluation=::NewEnumerator0
      Recalculate Random Each Loop:NiagaraBool=false
      Randomness Mode:ENiagaraRandomnessMode=::NewEnumerator0
      Random Seed:NiagaraInt32=?
      Fixed Random Seed:NiagaraBool=false
      Bool Selection Array:NiagaraDataInterfaceArrayBool=DataInterface() {propertyValues {GpuSyncMode=SyncCpuToGpu, MaxElements=0, BoolData [false,false,false,false,false,false,true,true,true,true,true,true,true,true]}}
     )
     True Float=-1
     False Float=1
    )) {isSetParameters=true, module=SetVariables_#7}
   /SetVariables(
    Particles.Offset:Vector3f=&MakeVector(
     X=0
     Y=&Multiply_Float(
      A=$Particles.Mirror
      B=&SelectFloatFromArray(
       Direct Array Index=$Particles.PartitionParticles.Partition
       Direct Array Mode=::NewEnumerator1
       Float Selection Array=DataInterface() {propertyValues {FloatData [0,0,0,0,0,0,0]}}
      )
     )
     Z=0
    )
    Particles.TravelDistance:NiagaraFloat=&Multiply_Float(
     A=&Multiply_Float(
      A=$Particles.Mirror
      B=&SelectFloatFromArray(
       Direct Array Index=$Particles.PartitionParticles.Partition
       Direct Array Mode=::NewEnumerator1
       Float Selection Array=DataInterface() {propertyValues {FloatData [0,0,0,0,0,0,0]}}
      )
     )
     B=&RandomRangeFloat(Minimum=1, Maximum=1)
    )
   ) {isSetParameters=true, module=SetVariables_#8}
   &InitializeParticle(
    Color=<rgba 1 0.1528 0.3162 1>
    Lifetime=&RandomRangeFloat(Minimum=0.5, Maximum=0.8)
    Mass=300
    Mass Mode=::NewEnumerator1
    Mesh Renderer Array Visibility Mode=::NewEnumerator3
    Mesh Renderer Info=DataInterface() {propertyValues {MeshRenderer=&self:NiagaraEmitter_1.NiagaraMeshRendererProperties_0}}
    Mesh Scale=&MakeVector(
     X=1
     Y=&RandomRangeFloat(Minimum=2, Maximum=4)
     Z=&RandomRangeFloat(Minimum=2, Maximum=4)
    )
    Mesh Scale Mode=::NewEnumerator4
    Sprite Size Mode=::NewEnumerator3
    Uniform Sprite Size=500
    UsePositionOffset=false
    Write Material Random=true
   )
   &InitialMeshOrientation(
    Mesh Orientation Mode=::NewEnumerator2
    Orientation Vector=<-1 0 0>
    Rotation=&RandomRangeVector(Minimum=<0 0 0>, Maximum=<1 0 0>)
    Use Rotation Vector=true
   )
  ]
  particleUpdateScript [
   &ParticleState()
   /Niagara/Modules/Spawn/Location/StaticMeshLocation(
    MaskSpawnGroup=false
    OffsetAlongNormal=false
    OverrideRandomSeed=false
    SamplePosOffset=false
    Sampled Position Offset:Vector3f=&Add_Vector(
     A=&MakeVector(X=&Multiply_Float(A=$Particles.NormalizedAge, B=$Particles.TravelDistance), Y=0, Z=0)
     B=&RotateVectorByQuaternion(
      Vector To Rotate=$Particles.Offset
      Quaternion=&MakeQuaternion(
       Quaternion From=0
       Axis=<1 0 0>
       Angle=&Waveform(
        Global Phase=&Add_Float(A=$Particles.Age, B=&RandomRangeFloat(Minimum=0, Maximum=1))
        Global Amplitude Scale=0.2
        "Waveform [1]"=::NewEnumerator0
        "[1] Frequency"=&RandomRangeFloat(Minimum=0.7, Maximum=1.2)
        "[1] Amplitude Min/Max"=<-1 1>
       )
       From=?
       To=?
       Angle Type=::NewEnumerator1
       Normalize Vectors=false
      )
     )
    )
    Static Mesh=DataInterface() {propertyValues {
      DefaultMesh=/DragonCartNarrative/Assets/Props/Fragments/SM_DragonCart_Fragment2
      InstanceIndex=-1
      LODIndex=0
      PreviewMesh=&SM_DragonCart_Fragment2
      SoftSourceActor=null
      SourceMode=Default
      bAllowSamplingFromStreamingLODs=false
      bCaptureTransformsPerFrame=true
      bUsePhysicsBodyVelocity=false
      FilteredSockets []
      LODIndexUserParameter {parameter {name=None, typeDefHandle {registeredTypeIndex=108}, varData []}}
      MeshParameterBinding {parameter {name=None, typeDefHandle {registeredTypeIndex=1}, varData []}}
      SectionFilter {allowedMaterialSlots []}
     }}
   )
   &SkeletalMeshLocation(
    Bone / Socket Index=$Particles.PartitionParticles.Partition
    Bone Sampling Mode=::NewEnumerator4
    MaskSpawnGroup=false
    Orientation Sampling=::NewEnumerator1
    Sampled Position Offset=&Add_Vector(
     A=&MakeVector(X=&Multiply_Float(A=$Particles.NormalizedAge, B=$Particles.TravelDistance), Y=0, Z=0)
     B=&RotateVectorByQuaternion(
      Vector To Rotate=$Particles.Offset
      Quaternion=&MakeQuaternion(
       Quaternion From=0
       Axis=<1 0 0>
       Angle=&Waveform(
        Global Phase=&Add_Float(A=$Particles.Age, B=&RandomRangeFloat(Minimum=0, Maximum=1))
        Global Amplitude Scale=0.2
        "Waveform [1]"=::NewEnumerator0
        "[1] Frequency"=&RandomRangeFloat(Minimum=0.7, Maximum=1.2)
        "[1] Amplitude Min/Max"=<-1 1>
       )
       From=?
       To=?
       Angle Type=::NewEnumerator1
       Normalize Vectors=false
      )
     )
    )
    Skeletal Mesh=$System.Skeletal Mesh Reader
   ) {enabled=false}
   /SetVariables(Particles.LightPosition:NiagaraPosition=&AddVectorToPosition(
     Position=$Particles.Position
     Vector=&Multiply_VectorByFloat(
      Vector=&ReturnCameraVector(
       Camera Vectors=::NewEnumerator1
       Coordinate Space=::Simulation
       CameraQuery=DataInterface() {propertyValues {bRequireCurrentFrameData=false}}
      )
      Float=-400
     )
    )) {isSetParameters=true, module=SetVariables_#9}
   /SetVariables(Particles.StartPosition:NiagaraPosition=&Lerp_Vector(
     A=Linked() {linkedVariable {name=Engine.Owner.Position, type=NiagaraPosition}}
     B=Linked() {linkedVariable {name=Particles.Position, type=NiagaraPosition}}
     Alpha=0.1
    )) {isSetParameters=true, module=SetVariables_#10}
   /SetVariables(
    Particles.Flare:NiagaraFloat=&FloatFromCurve(FloatCurve=DataInterface() {propertyValues {Curve {keys [
         {arriveTangent=-9.9823, interpMode=RCIM_Cubic, leaveTangent=-9.9823, tangentMode=RCTM_User, time=0.0073, value=1}
         {arriveTangentWeight=0.1021, interpMode=RCIM_Cubic, leaveTangentWeight=0.2548, tangentMode=RCTM_User, time=0.15, value=0}
         {
          arriveTangentWeight=1.9969
          interpMode=RCIM_Cubic
          leaveTangent=58.3845
          leaveTangentWeight=0.804
          tangentMode=RCTM_Break
          time=0.35
          value=0
         }
         {arriveTangent=-4.9513, interpMode=RCIM_Cubic, leaveTangent=-4.9513, tangentMode=RCTM_User, time=0.4, value=1}
         <at 0.7 0>
        ]}}})
    Particles.Flicker:NiagaraFloat=&Waveform(
     Global Input Position=&RandomRangeFloat(Minimum=0, Maximum=1)
     Global Phase=$Particles.Age
     "Waveform [1]"=::NewEnumerator0
     "[1] Phase"=$Particles.Age
     "[1] Frequency"=15
     "[1] Amplitude Min/Max"=<0.5 1>
    )
   ) {isSetParameters=true, module=SetVariables_#11}
   &ScaleMeshSize(Scale Factor=&VectorFromCurve(
     VectorCurve=DataInterface() {propertyValues {
       XCurve {keys [
         {interpMode=RCIM_Cubic, leaveTangentWeight=0.1333, tangentMode=RCTM_User, time=0.6, value=1}
         {arriveTangent=-2.2924, tangentMode=RCTM_User, time=1, value=0.7}
        ]}
       YCurve {keys [
         {arriveTangent=3.3015, interpMode=RCIM_Cubic, leaveTangent=3.3015, tangentMode=RCTM_User, time=0, value=0.25}
         {arriveTangentWeight=0.1954, interpMode=RCIM_Cubic, leaveTangentWeight=0.1466, tangentMode=RCTM_User, time=0.4, value=1}
        ]}
       ZCurve {keys [
         {arriveTangent=3.3015, interpMode=RCIM_Cubic, leaveTangent=3.3015, tangentMode=RCTM_User, time=0, value=0.25}
         {arriveTangentWeight=0.1954, interpMode=RCIM_Cubic, leaveTangentWeight=0.1466, tangentMode=RCTM_User, time=0.4, value=1}
        ]}
      }}
     Scale Curve=&MakeVector(
      X=&Divide_Float(A=&DistanceBetweenPositions(A=$Particles.Position, B=$Particles.StartPosition), B=2000)
      Y=&Add_Float(A=1, B=$Particles.Flare)
      Z=&Add_Float(A=1, B=$Particles.Flare)
     )
    ))
   &ScaleSpriteSize(
    Uniform Curve Scale=&Lerp_Float(A=0.5, B=1, Alpha=$Particles.Flicker)
    Uniform Curve Sprite Scale=DataInterface() {propertyValues {Curve {keys [
        {arriveTangent=60.0118, interpMode=RCIM_Cubic, leaveTangent=60.0118, tangentMode=RCTM_User, time=0, value=0}
        {arriveTangent=-5.251, interpMode=RCIM_Cubic, leaveTangent=-5.251, tangentMode=RCTM_User, time=0.1, value=2}
        {arriveTangentWeight=0.0833, interpMode=RCIM_Cubic, leaveTangentWeight=0.3333, tangentMode=RCTM_User, time=0.25, value=1}
        {arriveTangentWeight=0.3449, interpMode=RCIM_Cubic, leaveTangentWeight=0.0805, tangentMode=RCTM_User, time=0.7, value=1}
        {arriveTangent=-15.6866, tangentMode=RCTM_User, time=0.8, value=0}
       ]}}}
   )
   /SetVariables(Particles.SpriteSizeEnd:Vector2f=&Vector2DFromFloat(Value=&Multiply_Float(
      A=&FloatFromCurve(
       FloatCurve=DataInterface() {propertyValues {Curve {keys [
           {arriveTangent=65.4479, interpMode=RCIM_Cubic, leaveTangent=65.4479, tangentMode=RCTM_User, time=0.3, value=0}
           {arriveTangent=-5.251, interpMode=RCIM_Cubic, leaveTangent=-5.251, tangentMode=RCTM_User, time=0.4, value=2}
           {arriveTangentWeight=0.1383, interpMode=RCIM_Cubic, leaveTangentWeight=0.2421, tangentMode=RCTM_User, time=0.6, value=1}
           {arriveTangentWeight=0.3449, interpMode=RCIM_Cubic, leaveTangentWeight=0.0805, tangentMode=RCTM_User, time=0.85, value=1}
           {arriveTangent=-7.8926, tangentMode=RCTM_User, time=1, value=0}
          ]}}}
       Scale Curve=&Lerp_Float(A=0.5, B=1, Alpha=$Particles.Flicker)
      )
      B=1000
     ))) {isSetParameters=true, module=SetVariables_#12}
   &ScaleColor(
    Scale Alpha=&FloatFromCurve(FloatCurve=DataInterface() {propertyValues {Curve {keys [<at 0.4 1 RCIM_Cubic>,<at 1 0 RCIM_Cubic>]}}})
    Scale RGB=&VectorFromFloat(Value=&Multiply_Float(A=$Particles.Flicker, B=&Lerp_Float(A=1, B=1.5, Alpha=$Particles.Flare)))
    ScaleA=true
    ScaleRGB=true
   )
   /SetVariables(Particles.FlareColor=&Multiply_LinearColorByFloat(
     Linear Color=$Emitter.LightningColor
     Float=&FloatFromCurve(
      FloatCurve=DataInterface() {propertyValues {Curve {
         defaultValue=1
         keys [
          {arriveTangent=45.2252, interpMode=RCIM_Cubic, leaveTangent=45.2252, tangentMode=RCTM_User, time=0, value=0}
          {arriveTangent=6.6667, interpMode=RCIM_Cubic, leaveTangent=6.6667, time=0.05, value=1.5}
          {arriveTangentWeight=0.0409, interpMode=RCIM_Cubic, leaveTangentWeight=0.1789, tangentMode=RCTM_User, time=0.15, value=1}
          {interpMode=RCIM_Cubic, tangentMode=RCTM_User, time=0.8, value=1}
          {arriveTangent=-8.0179, tangentMode=RCTM_User, time=1, value=0}
         ]
        }}}
      Scale Curve=&Multiply_Float(A=$Particles.Flicker, B=&Lerp_Float(A=2.5, B=4, Alpha=$Particles.Flare))
     )
    )) {isSetParameters=true, module=SetVariables_#13}
   &CameraOffset(Camera Offset Amount=&Add_Float(
     A=/Niagara/DynamicInputs/Math/CalculateParticleRadius(
      Radius Calculation Type:ENiagaraCollisionRadiusOptions=::NewEnumerator0
      Particle Radius:NiagaraFloat=1
      Mesh Dimensions:Vector3f=<1 1 1>
      Method for Calculating Particle Radius:ENiagaraCalculateRadiusOptions=::NewEnumerator2
     )
     B=160
    ))
   &UpdateMeshOrientation(
    Facing Direction=&Subtract_Position(A=$Particles.Position, B=$Particles.StartPosition)
    Facing Interpolation=false
    Orientation Method=::NewEnumerator1
    Override Quat=false
    TransformFacing=false
    TransformUp=false
   ) {module=UpdateMeshOrientation001}
   &UpdateMeshOrientation(Rotation Rate=0)
   &DynamicMaterialParameters(
    Index 0 Param 1=&FloatFromCurve(FloatCurve=DataInterface() {propertyValues {Curve {
        defaultValue=0
        keys [
         {arriveTangent=-0.8644, interpMode=RCIM_Cubic, leaveTangent=-0.8644, tangentMode=RCTM_User, time=0, value=1}
         {arriveTangent=-3.8427, tangentMode=RCTM_User, time=0.35, value=0.5}
         <at 0.65 0.5>
         {tangentMode=RCTM_User, time=1, value=0}
        ]
       }}})
    Index 0 Param 2=&Lerp_Float(A=0.3, B=0.4, Alpha=$Particles.Flare)
    Index 0 Param 3=&Lerp_Float(A=0.07, B=0.12, Alpha=$Particles.Flare)
    Index 0 Param 4=$Particles.MaterialRandom
    Index 1 Param 1=&FloatFromCurve(
     FloatCurve=DataInterface() {propertyValues {Curve {defaultValue=1, keys [{tangentMode=RCTM_User, time=0, value=0}, {tangentMode=RCTM_User, time=1, value=1}]}}}
     Scale Curve=4
    )
    Index 1 Param 2=25
    Index 1 Param 3=0.5
    Index 1 Param 4=&MakeCustomFloatFromBool(
     Bool=&Boolean_OrOperation(
      Bool1=&SetBoolByIntComparison(A=$Particles.PartitionParticles.Partition, B=2, Comparison Type=::NewEnumerator4)
      Bool2=&SetBoolByIntComparison(A=$Particles.PartitionParticles.Partition, B=9, Comparison Type=::NewEnumerator4)
     )
     True Float=-100
     False Float=100
    )
    Index 2 Param 1=400
    Index 3 Param 2=1
    Module.Index0Float0Write=true
    Module.Index0Float1Write=true
    Module.Index0Float2Write=true
    Module.Index0Float3Write=true
    Module.Index1Float0Write=true
    Module.Index1Float1Write=true
    Module.Index1Float2Write=true
    Module.Index1Float3Write=true
    Module.Index2Float0Write=true
    Module.Index2Float1Write=false
    Module.Index2Float2Write=false
    Module.Index2Float3Write=false
    Module.Index3Float0Write=false
    Module.Index3Float1Write=false
    Module.Index3Float2Write=false
    Module.Index3Float3Write=false
    Write Parameter Index 1=::NewEnumerator1
    Write Parameter Index 2=::NewEnumerator1
    Write Parameter Index 3=::NewEnumerator1
   )
   /SetVariables(Particles.LightColor=&Multiply_LinearColorByFloat(Linear Color=$Particles.FlareColor, Float=4)) {isSetParameters=true, module=SetVariables_#14}
   /SetVariables(Particles.LightRadius=&MakeFloatFromVector2d(Vector 2D=$Particles.SpriteSizeEnd)) {isSetParameters=true, module=SetVariables_#15}
   &CalculateAccurateVelocity()
  ]
  rendererClasses [
   &NiagaraMeshRendererProperties
   &NiagaraSpriteRendererProperties
   &NiagaraLightRendererProperties
  ]
  renderers [
   {rendererClass=&NiagaraMeshRendererProperties, rendererIndex=0}
   {rendererClass=&NiagaraSpriteRendererProperties, rendererIndex=1}
   {rendererClass=&NiagaraSpriteRendererProperties, rendererIndex=2}
   {rendererClass=&NiagaraLightRendererProperties, rendererIndex=3}
  ]
 }
 MedLightning001 {
  bEnabled=false
  emitterSpawnScript [/SetVariables(Emitter.LightningColor:LinearColor=&MakeCustomLinearColorFromBool(Bool=$User.IsRed, True Color=<rgba 0.0945 0.2938 1 1>, False Color=<rgba 0.873 0 1 1>)) {isSetParameters=true, module=SetVariables_#6}]
  emitterUpdateScript [
   &EmitterState()
   &SpawnBurst_Instantaneous(Spawn Count=8, Use Spawn Probability=false, UseLoopCountLimit=true) {enabled=false}
   &SpawnRate(SpawnRate=4, Use Spawn Probability=false)
  ]
  particleSpawnScript [
   &PartitionParticles(Number Of Partitions=14)
   /SetVariables(Particles.Mirror:NiagaraFloat=&MakeCustomFloatFromBool(
     Bool=&SelectBoolFromArray(
      Array Sampling Mode:ENiagara_LifetimeMode=::NewEnumerator0
      Direct Array Index:NiagaraInt32=$Particles.PartitionParticles.Partition
      Direct Array Mode:ENiagara_WrapClamp=::NewEnumerator1
      Evaluation Type:ENiagaraRandomnessEvaluation=::NewEnumerator0
      Recalculate Random Each Loop:NiagaraBool=false
      Randomness Mode:ENiagaraRandomnessMode=::NewEnumerator0
      Random Seed:NiagaraInt32=?
      Fixed Random Seed:NiagaraBool=false
      Bool Selection Array:NiagaraDataInterfaceArrayBool=DataInterface() {propertyValues {GpuSyncMode=SyncCpuToGpu, MaxElements=0, BoolData [false,false,false,false,false,false,true,true,true,true,true,true,true,true]}}
     )
     True Float=-1
     False Float=1
    )) {isSetParameters=true, module=SetVariables_#7}
   /SetVariables(
    Particles.Offset:Vector3f=&MakeVector(
     X=0
     Y=&Multiply_Float(
      A=$Particles.Mirror
      B=&SelectFloatFromArray(
       Direct Array Index=$Particles.PartitionParticles.Partition
       Direct Array Mode=::NewEnumerator1
       Float Selection Array=DataInterface() {propertyValues {FloatData [400,300,300,250,200,150,-800]}}
      )
     )
     Z=0
    )
    Particles.TravelDistance:NiagaraFloat=&Multiply_Float(
     A=&Multiply_Float(
      A=$Particles.Mirror
      B=&SelectFloatFromArray(
       Direct Array Index=$Particles.PartitionParticles.Partition
       Direct Array Mode=::NewEnumerator1
       Float Selection Array=DataInterface() {propertyValues {FloatData [1600,1800,700,900,700,700,1100]}}
      )
     )
     B=&RandomRangeFloat(Minimum=1, Maximum=1)
    )
   ) {isSetParameters=true, module=SetVariables_#8}
   &InitializeParticle(
    Color=<rgba 1 0.1528 0.3162 1>
    Lifetime=&RandomRangeFloat(Minimum=0.5, Maximum=0.8)
    Mass=300
    Mass Mode=::NewEnumerator1
    Mesh Renderer Array Visibility Mode=::NewEnumerator3
    Mesh Renderer Info=DataInterface() {propertyValues {MeshRenderer=&self:MedLightning001_0.NiagaraMeshRendererProperties_0}}
    Mesh Scale=&MakeVector(
     X=1
     Y=&RandomRangeFloat(Minimum=2, Maximum=4)
     Z=&RandomRangeFloat(Minimum=2, Maximum=4)
    )
    Mesh Scale Mode=::NewEnumerator4
    Sprite Size Mode=::NewEnumerator3
    Uniform Sprite Size=200
    UsePositionOffset=false
    Write Material Random=true
   )
   &InitialMeshOrientation(
    Mesh Orientation Mode=::NewEnumerator2
    Orientation Vector=<-1 0 0>
    Rotation=&RandomRangeVector(Minimum=<0 0 0>, Maximum=<1 0 0>)
    Use Rotation Vector=true
   )
  ]
  particleUpdateScript [
   &ParticleState()
   &ShapeLocation(
    MaskSpawnGroup=false
    Override Seed=false
    Shape Origin=&Add_Vector(
     A=&MakeVector(X=&Multiply_Float(A=$Particles.NormalizedAge, B=$Particles.TravelDistance), Y=0, Z=0)
     B=&RotateVectorByQuaternion(
      Vector To Rotate=$Particles.Offset
      Quaternion=&MakeQuaternion(
       Quaternion From=0
       Axis=<1 0 0>
       Angle=&Waveform(
        Global Phase=&Add_Float(A=$Particles.Age, B=&RandomRangeFloat(Minimum=0, Maximum=1))
        Global Amplitude Scale=0.2
        "Waveform [1]"=::NewEnumerator0
        "[1] Frequency"=&RandomRangeFloat(Minimum=0.7, Maximum=1.2)
        "[1] Amplitude Min/Max"=<-1 1>
       )
       From=?
       To=?
       Angle Type=::NewEnumerator1
       Normalize Vectors=false
      )
     )
    )
    UseOwnerScale=true
   )
   &SkeletalMeshLocation(
    Bone / Socket Index=$Particles.PartitionParticles.Partition
    Bone Sampling Mode=::NewEnumerator4
    MaskSpawnGroup=false
    Orientation Sampling=::NewEnumerator1
    Sampled Position Offset=&Add_Vector(
     A=&MakeVector(X=&Multiply_Float(A=$Particles.NormalizedAge, B=$Particles.TravelDistance), Y=0, Z=0)
     B=&RotateVectorByQuaternion(
      Vector To Rotate=$Particles.Offset
      Quaternion=&MakeQuaternion(
       Quaternion From=0
       Axis=<1 0 0>
       Angle=&Waveform(
        Global Phase=&Add_Float(A=$Particles.Age, B=&RandomRangeFloat(Minimum=0, Maximum=1))
        Global Amplitude Scale=0.2
        "Waveform [1]"=::NewEnumerator0
        "[1] Frequency"=&RandomRangeFloat(Minimum=0.7, Maximum=1.2)
        "[1] Amplitude Min/Max"=<-1 1>
       )
       From=?
       To=?
       Angle Type=::NewEnumerator1
       Normalize Vectors=false
      )
     )
    )
    Skeletal Mesh=$System.Skeletal Mesh Reader
   ) {enabled=false}
   /SetVariables(Particles.LightPosition:NiagaraPosition=&AddVectorToPosition(
     Position=$Particles.Position
     Vector=&Multiply_VectorByFloat(
      Vector=&ReturnCameraVector(
       Camera Vectors=::NewEnumerator1
       Coordinate Space=::Simulation
       CameraQuery=DataInterface() {propertyValues {bRequireCurrentFrameData=false}}
      )
      Float=-400
     )
    )) {isSetParameters=true, module=SetVariables_#9}
   /SetVariables(Particles.StartPosition:NiagaraPosition=&Lerp_Vector(
     A=Linked() {linkedVariable {name=Engine.Owner.Position, type=NiagaraPosition}}
     B=Linked() {linkedVariable {name=Particles.Position, type=NiagaraPosition}}
     Alpha=0.1
    )) {isSetParameters=true, module=SetVariables_#10}
   /SetVariables(
    Particles.Flare:NiagaraFloat=&FloatFromCurve(FloatCurve=DataInterface() {propertyValues {Curve {keys [
         {arriveTangent=-9.9823, interpMode=RCIM_Cubic, leaveTangent=-9.9823, tangentMode=RCTM_User, time=0.0073, value=1}
         {arriveTangentWeight=0.1021, interpMode=RCIM_Cubic, leaveTangentWeight=0.2548, tangentMode=RCTM_User, time=0.15, value=0}
         {
          arriveTangentWeight=1.9969
          interpMode=RCIM_Cubic
          leaveTangent=58.3845
          leaveTangentWeight=0.804
          tangentMode=RCTM_Break
          time=0.35
          value=0
         }
         {arriveTangent=-4.9513, interpMode=RCIM_Cubic, leaveTangent=-4.9513, tangentMode=RCTM_User, time=0.4, value=1}
         <at 0.7 0>
        ]}}})
    Particles.Flicker:NiagaraFloat=&Waveform(
     Global Input Position=&RandomRangeFloat(Minimum=0, Maximum=1)
     Global Phase=$Particles.Age
     "Waveform [1]"=::NewEnumerator0
     "[1] Phase"=$Particles.Age
     "[1] Frequency"=15
     "[1] Amplitude Min/Max"=<0.5 1>
    )
   ) {isSetParameters=true, module=SetVariables_#11}
   &ScaleMeshSize(Scale Factor=&VectorFromCurve(
     VectorCurve=DataInterface() {propertyValues {
       XCurve {keys [
         {interpMode=RCIM_Cubic, leaveTangentWeight=0.1333, tangentMode=RCTM_User, time=0.6, value=1}
         {arriveTangent=-2.2924, tangentMode=RCTM_User, time=1, value=0.7}
        ]}
       YCurve {keys [
         {arriveTangent=3.3015, interpMode=RCIM_Cubic, leaveTangent=3.3015, tangentMode=RCTM_User, time=0, value=0.25}
         {arriveTangentWeight=0.1954, interpMode=RCIM_Cubic, leaveTangentWeight=0.1466, tangentMode=RCTM_User, time=0.4, value=1}
        ]}
       ZCurve {keys [
         {arriveTangent=3.3015, interpMode=RCIM_Cubic, leaveTangent=3.3015, tangentMode=RCTM_User, time=0, value=0.25}
         {arriveTangentWeight=0.1954, interpMode=RCIM_Cubic, leaveTangentWeight=0.1466, tangentMode=RCTM_User, time=0.4, value=1}
        ]}
      }}
     Scale Curve=&MakeVector(
      X=&Divide_Float(A=&DistanceBetweenPositions(A=$Particles.Position, B=$Particles.StartPosition), B=1000)
      Y=&Add_Float(A=1, B=$Particles.Flare)
      Z=&Add_Float(A=1, B=$Particles.Flare)
     )
    ))
   &ScaleSpriteSize(
    Uniform Curve Scale=&Lerp_Float(A=0.5, B=1, Alpha=$Particles.Flicker)
    Uniform Curve Sprite Scale=DataInterface() {propertyValues {Curve {keys [
        {arriveTangent=60.0118, interpMode=RCIM_Cubic, leaveTangent=60.0118, tangentMode=RCTM_User, time=0, value=0}
        {arriveTangent=-5.251, interpMode=RCIM_Cubic, leaveTangent=-5.251, tangentMode=RCTM_User, time=0.1, value=2}
        {arriveTangentWeight=0.0833, interpMode=RCIM_Cubic, leaveTangentWeight=0.3333, tangentMode=RCTM_User, time=0.25, value=1}
        {arriveTangentWeight=0.3449, interpMode=RCIM_Cubic, leaveTangentWeight=0.0805, tangentMode=RCTM_User, time=0.7, value=1}
        {arriveTangent=-15.6866, tangentMode=RCTM_User, time=0.8, value=0}
       ]}}}
   )
   /SetVariables(Particles.SpriteSizeEnd:Vector2f=&Vector2DFromFloat(Value=&Multiply_Float(
      A=&FloatFromCurve(
       FloatCurve=DataInterface() {propertyValues {Curve {keys [
           {arriveTangent=65.4479, interpMode=RCIM_Cubic, leaveTangent=65.4479, tangentMode=RCTM_User, time=0.3, value=0}
           {arriveTangent=-5.251, interpMode=RCIM_Cubic, leaveTangent=-5.251, tangentMode=RCTM_User, time=0.4, value=2}
           {arriveTangentWeight=0.1383, interpMode=RCIM_Cubic, leaveTangentWeight=0.2421, tangentMode=RCTM_User, time=0.6, value=1}
           {arriveTangentWeight=0.3449, interpMode=RCIM_Cubic, leaveTangentWeight=0.0805, tangentMode=RCTM_User, time=0.85, value=1}
           {arriveTangent=-7.8926, tangentMode=RCTM_User, time=1, value=0}
          ]}}}
       Scale Curve=&Lerp_Float(A=0.5, B=1, Alpha=$Particles.Flicker)
      )
      B=500
     ))) {isSetParameters=true, module=SetVariables_#12}
   &ScaleColor(
    Scale Alpha=&FloatFromCurve(FloatCurve=DataInterface() {propertyValues {Curve {keys [<at 0.4 1 RCIM_Cubic>,<at 1 0 RCIM_Cubic>]}}})
    Scale RGB=&VectorFromFloat(Value=&Multiply_Float(A=$Particles.Flicker, B=&Lerp_Float(A=1, B=1.5, Alpha=$Particles.Flare)))
    ScaleA=true
    ScaleRGB=true
   )
   /SetVariables(Particles.FlareColor=&Multiply_LinearColorByFloat(
     Linear Color=$Emitter.LightningColor
     Float=&FloatFromCurve(
      FloatCurve=DataInterface() {propertyValues {Curve {
         defaultValue=1
         keys [
          {arriveTangent=45.2252, interpMode=RCIM_Cubic, leaveTangent=45.2252, tangentMode=RCTM_User, time=0, value=0}
          {arriveTangent=6.6667, interpMode=RCIM_Cubic, leaveTangent=6.6667, time=0.05, value=1.5}
          {arriveTangentWeight=0.0409, interpMode=RCIM_Cubic, leaveTangentWeight=0.1789, tangentMode=RCTM_User, time=0.15, value=1}
          {interpMode=RCIM_Cubic, tangentMode=RCTM_User, time=0.8, value=1}
          {arriveTangent=-8.0179, tangentMode=RCTM_User, time=1, value=0}
         ]
        }}}
      Scale Curve=&Multiply_Float(A=$Particles.Flicker, B=&Lerp_Float(A=2.5, B=4, Alpha=$Particles.Flare))
     )
    )) {isSetParameters=true, module=SetVariables_#13}
   &CameraOffset(Camera Offset Amount=&Add_Float(
     A=&CalculateParticleRadius(
      Radius Calculation Type:ENiagaraCollisionRadiusOptions=::NewEnumerator0
      Particle Radius:NiagaraFloat=1
      Mesh Dimensions:Vector3f=<1 1 1>
      Method for Calculating Particle Radius:ENiagaraCalculateRadiusOptions=::NewEnumerator2
     )
     B=160
    ))
   &UpdateMeshOrientation(
    Facing Direction=&Subtract_Position(A=$Particles.Position, B=$Particles.StartPosition)
    Facing Interpolation=false
    Orientation Method=::NewEnumerator1
    Override Quat=false
    TransformFacing=false
    TransformUp=false
   ) {module=UpdateMeshOrientation001}
   &UpdateMeshOrientation(Rotation Rate=0)
   &DynamicMaterialParameters(
    Index 0 Param 1=&FloatFromCurve(FloatCurve=DataInterface() {propertyValues {Curve {
        defaultValue=0
        keys [
         {arriveTangent=-0.8644, interpMode=RCIM_Cubic, leaveTangent=-0.8644, tangentMode=RCTM_User, time=0, value=1}
         {arriveTangent=-3.8427, tangentMode=RCTM_User, time=0.35, value=0.5}
         <at 0.65 0.5>
         {tangentMode=RCTM_User, time=1, value=0}
        ]
       }}})
    Index 0 Param 2=&Lerp_Float(A=1, B=2, Alpha=$Particles.Flare)
    Index 0 Param 3=&Lerp_Float(A=0.01, B=0.02, Alpha=$Particles.Flare)
    Index 0 Param 4=$Particles.MaterialRandom
    Index 1 Param 1=&FloatFromCurve(FloatCurve=DataInterface() {propertyValues {Curve {defaultValue=1, keys [{tangentMode=RCTM_User, time=0, value=0}, {tangentMode=RCTM_User, time=1, value=1}]}}})
    Index 1 Param 2=1
    Index 1 Param 3=1
    Index 2 Param 1=400
    Index 3 Param 2=1
    Module.Index0Float0Write=true
    Module.Index0Float1Write=true
    Module.Index0Float2Write=true
    Module.Index0Float3Write=true
    Module.Index1Float0Write=true
    Module.Index1Float1Write=true
    Module.Index1Float2Write=true
    Module.Index1Float3Write=true
    Module.Index2Float0Write=true
    Module.Index2Float1Write=false
    Module.Index2Float2Write=false
    Module.Index2Float3Write=false
    Module.Index3Float0Write=false
    Module.Index3Float1Write=false
    Module.Index3Float2Write=false
    Module.Index3Float3Write=false
    Write Parameter Index 1=::NewEnumerator1
    Write Parameter Index 2=::NewEnumerator1
    Write Parameter Index 3=::NewEnumerator1
   )
   /SetVariables(Particles.LightColor=&Multiply_LinearColorByFloat(Linear Color=$Particles.FlareColor, Float=4)) {isSetParameters=true, module=SetVariables_#14}
   /SetVariables(Particles.LightRadius=&MakeFloatFromVector2d(Vector 2D=$Particles.SpriteSizeEnd)) {isSetParameters=true, module=SetVariables_#15}
   &CalculateAccurateVelocity()
  ]
  rendererClasses [
   &NiagaraMeshRendererProperties
   &NiagaraSpriteRendererProperties
   &NiagaraLightRendererProperties
  ]
  renderers [
   {rendererClass=&NiagaraMeshRendererProperties, rendererIndex=0}
   {rendererClass=&NiagaraSpriteRendererProperties, rendererIndex=1}
   {rendererClass=&NiagaraSpriteRendererProperties, rendererIndex=2}
   {rendererClass=&NiagaraLightRendererProperties, rendererIndex=3}
  ]
 }
}
summary {
 systemName=NS_NDC_DragonCart_Shard_Eletrify
 emitters [
  {
   bEnabled=false
   emitterName=ElectricArcs
   rendererClasses [&NiagaraMeshRendererProperties, &NiagaraSpriteRendererProperties]
  }
  {
   bEnabled=false
   emitterName=ElectricArcs001
   rendererClasses [&NiagaraMeshRendererProperties, &NiagaraSpriteRendererProperties]
  }
  {
   emitterName=ElectricArcs002
   rendererClasses [&NiagaraMeshRendererProperties, &NiagaraSpriteRendererProperties]
  }
  {
   bEnabled=false
   emitterName=MedLightning
   rendererClasses [
    &NiagaraMeshRendererProperties
    &NiagaraSpriteRendererProperties
    &NiagaraLightRendererProperties
   ]
  }
  {
   bEnabled=false
   emitterName=ElectricArcs003
   rendererClasses [&NiagaraMeshRendererProperties, &NiagaraSpriteRendererProperties]
  }
  {
   bEnabled=false
   emitterName=ElectricArcs004
   rendererClasses [&NiagaraMeshRendererProperties, &NiagaraSpriteRendererProperties]
  }
  {
   bEnabled=false
   emitterName=MedLightning001
   rendererClasses [
    &NiagaraMeshRendererProperties
    &NiagaraSpriteRendererProperties
    &NiagaraLightRendererProperties
   ]
  }
 ]
 userVariables [
  User.Buildup {type=NiagaraFloat, defaultValue=0, description=""}
  User.IsRed {type=NiagaraBool, defaultValue=true}
  User.LocalDestination {type=NiagaraPosition, description="", defaultValue {}}
 ]
}
